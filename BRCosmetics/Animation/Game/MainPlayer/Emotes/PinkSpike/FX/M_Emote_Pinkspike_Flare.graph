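# Additive unlit Niagara sprite material for a flare effect in the PinkSpike emote.
# Particle color tints the red channel of a shared sparkle texture, which then passes through the MF_FX_EmissiveAdjustments function.
# Particle alpha is scaled by an Opacity scalar parameter to control the flare's strength.
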Material /BRCosmetics/Animation/Game/MainPlayer/Emotes/PinkSpike/FX/M_Emote_Pinkspike_Flare
graph {
 properties {bUsedWithNiagaraSprites=true, blendMode=BLEND_Additive, pixelDepthOffsetMode=PDOM_Legacy, shadingModel=MSM_Unlit, usedShadingModels=Unlit}
 nodes [
  n1 = ParticleColor()
  MaterialFunctionCall(Emissive=Multiply(A=@n1.RGB, B=TextureSample() {texture=/Game/Effects/Fort_Effects/Textures/Sparkle/T_Sparkle_01}.R)) {materialFunction=&MF_FX_EmissiveAdjustments}
  Multiply(A=@n1.A, B=ScalarParameter() {defaultValue=0.8, parameterName=Opacity})
 ]
}
