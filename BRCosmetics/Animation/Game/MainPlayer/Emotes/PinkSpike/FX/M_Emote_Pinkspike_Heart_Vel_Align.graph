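# Translucent unlit Niagara sprite material for the heart effect in the PinkSpike emote, using a randomly picked frame from a two-frame SubUV texture.
# Smoothstepped normal-map lighting against the sun vector shades particle color, blended with a time-of-day dark tint by the texture's green channel.
# A WPO graph rebuilds sprite corners in view space to align with projected particle direction, fading toward default orientation as that direction parallels the camera's horizontally.
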Material /BRCosmetics/Animation/Game/MainPlayer/Emotes/PinkSpike/FX/M_Emote_Pinkspike_Heart_Vel_Align
graph {
 properties {bUsedWithNiagaraSprites=true, blendMode=BLEND_Translucent, pixelDepthOffsetMode=PDOM_Legacy, shadingModel=MSM_Unlit, usedShadingModels=Unlit}
 nodes [
  n1 = Constant3Vector() {constant=<rgba 0.0694 0.0166 0.1146 1>}
  n2 = Constant3Vector() {constant=<rgba 0.0158 0.0038 0.026 1>}
  n3 = ParticleColor()
  n4 = ScalarParameter() {parameterName=Smoothstep Amount}
  n5 = CollectionParameter() {
   collection=&FortniteMaterialParameters
   parameterName=SunlightDirectionalVector
  }
  n6 = NamedRerouteUsage()
  n7 = DotProduct(
   A=Normalize(VectorInput=@n5)
   B=Transform(None=TextureSample(UVs=@n6) {samplerType=SAMPLERTYPE_Normal, texture=&/T_Emote_PinkSpike_Heart_N}.RGB)
  )
  n8 = SmoothStep(
   Min=@n4
   Max=Add(A=@n4, B=ScalarParameter() {defaultValue=0.02, parameterName=Smoothstep Falloff})
   Value=Saturate(None=Multiply(A=@n7) {constB=-1})
  )
  n9 = LinearInterpolate(A=Multiply(A=@n3.RGB) {constB=0.8}, B=@n3.RGB, Alpha=OneMinus(None=Saturate(None=@n8)))
  n10 = NamedRerouteDeclaration(None=MaterialFunctionCall!(
    U_Images=Constant() {r=2}
    V_Images=Constant() {r=1}
    SubUV Texture RGB
    Time Speed=Constant() {r=1}
    Manual Time Control=StaticBool() {value=true}
    Time=DynamicParameter() {paramNames [SubUV Random, Param2, Param3, Param4]}.SubUV Random
    UVs=TextureCoordinate()
   ) {materialFunction=/Game/Effects/Fort_Effects/Materials/_Functions/MF_SubUV}.UVs) {name=SubUV, nodeColor=<rgba 1 0 0.4922 1>}
  MaterialFunctionCall(Emissive=LinearInterpolate(
    A=MaterialFunctionCall(Color_Morning=@n1, Color_Day=@n1, Color_Evening=@n2, Color_Night=@n1) {materialFunction=&Color_Time_of_Day}.Result
    B=Saturate(None=@n9)
    Alpha=TextureSample(UVs=@n10) {texture=&/T_Emote_PinkSpike_Heart}.G
   )) {materialFunction=&MF_FX_EmissiveAdjustments}
  n11 = ParticlePositionWS()
  n12 = WorldPosition()
  n13 = ComponentMask(None=TextureCoordinate()) {r=true}
  n14 = ParticleSize()
  n15 = Multiply(
   A=Transform(None=ParticleDirection()) {transformSourceType=TRANSFORMSOURCE_World, transformType=TRANSFORM_View}
   B=Constant3Vector() {constant=<rgba 1 1 0 0>}
  )
  n16 = Multiply(A=Sign(None=ComponentMask(None=@n15) {r=true}), B=Normalize(VectorInput=@n15))
  n17 = ParticleDirection()
  n18 = Constant3Vector() {constant=<rgba 1 1 0 0>}
  n19 = MaterialFunctionCall() {materialFunction=&CameraDirectionVector}
  n20 = DotProduct(A=Normalize(VectorInput=Multiply(A=@n17, B=@n18)), B=Normalize(VectorInput=Multiply(A=@n19.Result, B=@n18)))
  n21 = SquareRoot(None=OneMinus(None=DotProduct(A=@n20, B=@n20)))
  n22 = LinearInterpolate(A=Constant3Vector() {constant=<rgba 1 0 0 0>}, B=@n16, Alpha=@n21)
  n23 = Multiply(
   A=Multiply(A=Subtract(A=@n13) {constB=0.5}, B=ComponentMask(None=@n14) {r=true})
   B=Normalize(VectorInput=Transform(None=@n22) {transformSourceType=TRANSFORMSOURCE_View})
  )
  n24 = ComponentMask(None=TextureCoordinate()) {g=true}
  n25 = Constant3Vector() {constant=<rgba 0 0 1 0>}
  n26 = LinearInterpolate(A=Constant3Vector() {constant=<rgba 0 -1 0 0>}, B=Normalize(VectorInput=CrossProduct(A=@n16, B=@n25)), Alpha=@n21)
  n27 = Multiply(
   A=Multiply(A=Subtract(A=@n24) {constB=0.5}, B=ComponentMask(None=@n14) {g=true})
   B=Normalize(VectorInput=Transform(None=@n26) {transformSourceType=TRANSFORMSOURCE_View})
  )
  NamedRerouteDeclaration(None=Add(A=Subtract(A=@n11, B=@n12.XYZ), B=Add(A=@n23, B=@n27))) {name=WPO, nodeColor=<rgba 0.1719 1 0 1>}
  NamedRerouteUsage()
 ]
}
