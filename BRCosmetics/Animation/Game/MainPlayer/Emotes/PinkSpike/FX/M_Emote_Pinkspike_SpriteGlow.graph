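# Additive unlit Niagara sprite material for a soft round glow in the PinkSpike emote effects, built without textures.
# A radial gradient from UV distance to center shifts the particle color's hue toward the edge, then passes through the shared emissive adjustments function.
# The squared inverse gradient times particle alpha forms the opacity falloff.
# A quality switch adds particle-radius depth fade except on Low and Medium; the Depth Fade Distance parameter is unconnected.
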Material /BRCosmetics/Animation/Game/MainPlayer/Emotes/PinkSpike/FX/M_Emote_Pinkspike_SpriteGlow
graph {
 properties {bUsedWithNiagaraSprites=true, blendMode=BLEND_Additive, pixelDepthOffsetMode=PDOM_Legacy, shadingModel=MSM_Unlit, usedShadingModels=Unlit}
 nodes [
  n1 = ScalarParameter() {defaultValue=0.1, parameterName=HueShift}
  n2 = TextureCoordinate()
  n3 = Constant() {r=0.5}
  n4 = Saturate(None=Multiply(A=Distance(A=@n2, B=@n3)) {constB=2})
  MaterialFunctionCall(Emissive=MaterialFunctionCall(Hue Shift Percentage=Multiply(A=@n1, B=@n4), Texture=ParticleColor().RGB) {materialFunction=&HueShift}.Result) {materialFunction=&MF_FX_EmissiveAdjustments}
  n5 = OneMinus(None=@n4)
  n6 = Multiply(A=Multiply(A=@n5, B=@n5), B=ParticleColor().A)
  n7 = DepthFade(Opacity=@n6, FadeDistance=ParticleRadius())
  QualitySwitch(Default=@n7, Low=@n6, High=@n7, Medium=@n6, Epic=@n7)
  ScalarParameter() {defaultValue=99, parameterName=Depth Fade Distance}
 ]
}
