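# Niagara system for the PinkSpike player emote, showing pulsing pink heart sprites with pops and flares.
# Hearts spawn in two groups with opposed up/down cone velocities, slow under drag, sink under ramped gravity and generate death events.
# Short-lived Pops (animated SubUV, glow) and Flare (wide streak) have no spawn modules shown.
# Time-of-day scalars dim colors, distance enlarges hearts, and a blue Hearts001 variant is disabled.
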NiagaraSystem /BRCosmetics/Animation/Game/MainPlayer/Emotes/PinkSpike/FX/NS_PinkSpike_Emote
dependencies {
 usedDataInterfaces [&NiagaraDataInterfaceCurve]
 usedDynamicInputs [&RandomRangeVector]
 usedModules [&SystemState]
}
emitters {
 Flare {
  emitterUpdateScript [&EmitterState(Life Cycle Mode=::NewEnumerator1, Loop Duration=0, Recalculate Duration Each Loop=true, UseLoopDelay=false)]
  particleSpawnScript [&InitializeParticle(
    Color=<rgba 0.9479 0.1432 0.5926 1>
    Lifetime=0.18
    Sprite Size Mode=::NewEnumerator3
    Uniform Sprite Size=8
    UsePositionOffset=false
    Write Material Random=true
   )]
  particleUpdateScript [
   &ParticleState()
   &ScaleSpriteSize(Uniform Curve Sprite Scale=DataInterface() {propertyValues {Curve {keys [
        {arriveTangent=-0.8906, interpMode=RCIM_Cubic, leaveTangent=-0.8906, tangentMode=RCTM_User, time=0, value=1}
        {tangentMode=RCTM_User, time=1, value=0}
       ]}}}) {module=ScaleSpriteSize001}
   /SetVariables(Particles.Color=&MultiplyLinearColor(
     Linear Color=$Particles.Initial.Color
     Scale Factor=&MakeVector4_FromLinearColor(LinearColor=&MakeLinearColorFromVectorAndFloat(
       "Vector (RGB)"=&VectorFromFloat(Value=&FNTimeOfDayScalar(Night Scale=1, Morning Scale=1, Evening Scale=0.5))
       "Float (Alpha)"=1
      ))
    )) {isSetParameters=true}
   /SetVariables(Particles.WideFlareSize:Vector2f=&Multiply_Vector2D(
     A=$Particles.SpriteSize
     B=&Multiply_Vector2D_ByFloat(
      Float=&FloatFromCurve(
       FloatCurve=DataInterface() {propertyValues {Curve {keys [
           {arriveTangent=9.0725, interpMode=RCIM_Cubic, leaveTangent=9.0725, tangentMode=RCTM_User, time=0, value=0}
           {arriveTangent=-2.5249, interpMode=RCIM_Cubic, leaveTangent=-2.5249, tangentMode=RCTM_User, time=0.3, value=0.9847}
           {arriveTangent=-0.0405, interpMode=RCIM_Cubic, leaveTangent=-0.0405, tangentMode=RCTM_User, time=1, value=0}
          ]}}}
       Scale Curve=1.8
      )
      Vector2D=<1.5 0.25>
     )
    )) {isSetParameters=true}
   &SolveForcesAndVelocity(Clamp Velocity=false, Limit Acceleration=false)
  ]
  rendererClasses [&NiagaraSpriteRendererProperties]
  renderers [
   {rendererClass=&NiagaraSpriteRendererProperties}
   {rendererClass=&NiagaraSpriteRendererProperties, rendererIndex=1}
  ]
 }
 Hearts {
  emitterUpdateScript [
   &EmitterState(Life Cycle Mode=::NewEnumerator1, Loop Delay=0.9083, Loop Duration=0, Recalculate Duration Each Loop=true, UseLoopDelay=false)
   &SpawnRate(SpawnRate=2.5, Use Spawn Probability=false) {module=SpawnRate001}
   &SpawnRate(SpawnGroup=1, SpawnRate=2.5, Use Spawn Probability=false)
  ]
  particleSpawnScript [
   &InitializeParticle(
    Color=&LerpLinearColors(
     StartColor=<rgba 0.974 0 1 1>
     EndColor=<rgba 0.9765 0.0938 1 1>
     LerpFactor=&MakeCustomFloatFromBool(Bool=&RandomBool(), True Float=1, False Float=0)
    )
    Lifetime=&RandomRangeFloat(Minimum=1.5, Maximum=2)
    Sprite Size=&Multiply_Vector2D_ByFloat(
     Float=&FloatFromCurve(
      FloatCurve=DataInterface() {propertyValues {Curve {keys [<at 0 0.65 RCIM_Constant>,<at 0.45 0.9 RCIM_Constant>,<at 0.9 1.3>]}}}
      CurveIndex=&RandomRangeFloat(Minimum=0, Maximum=1)
      Scale Curve=1.1
     )
     Vector2D=<3.5 3>
    )
    Sprite Size Mode=::NewEnumerator4
    UsePositionOffset=false
    Write Material Random=true
   )
   &SystemLocation(
    MaskSpawnGroup=false
    Offset=&Lerp_Vector(A=<-1 2 3.5>, B=<-1 2 -3.5>, Alpha=&MaskFloatBySpawnGroup(Masked Float=0, Passthrough Float=1))
   )
   /SetVariables(
    Particles.GlowColor=&MultiplyLinearColor(Linear Color=$Particles.Color, Scale Factor=&MakeVector4_FromFloat(Float=0.4))
    Particles.GlowSize=&Multiply_Vector2D_ByFloat(
     Float=1
     Vector2D=&Vector2DFromFloat(Value=&MakeFloatFromVector2d(Vector 2D=$Particles.SpriteSize))
    )
   ) {isSetParameters=true, module=SetVariables_#1}
   &AddVelocity(
    Cone Angle=60
    Cone Axis=&Lerp_Vector(A=<0 0 1>, B=<0 0 -1>, Alpha=&MaskFloatBySpawnGroup(Masked Float=0, Passthrough Float=1))
    Distribution Along Cone Axis=0.25
    Inner Cone Angle=10
    Module.UseConeFalloff=true
    Module.UseVelDistribution=true
    Override Seed=false
    Speed Falloff From Cone Axis=0.5
    Velocity Mode=::NewEnumerator2
    Velocity Speed=&Multiply_Float(A=&RandomRangeFloat(Minimum=15, Maximum=20), B=2)
   )
   &SubUVAnimation(End Frame=1, SubUV Animation Mode:ENiagara_SubUVLookupMode=::NewEnumerator2) {enabled=false}
   &DynamicMaterialParameters(
    Index 0 Param 1=&RandomRangeFloat(Minimum=0, Maximum=0.99)
    Module.Param0WriteEnabled=true
    Module.Param1WriteEnabled=true
    Module.Param2WriteEnabled=true
    Module.Param3WriteEnabled=true
   )
  ]
  particleUpdateScript [
   &ParticleState()
   &AccelerationForce(Acceleration=&MakeVector(
     X=0
     Y=0
     Z=&Lerp_Float(
      A=&FloatFromCurve(FloatCurve=DataInterface() {propertyValues {Curve {keys [<at 0 1.3 RCIM_Constant>,<at 0.2 -1.7 RCIM_Constant>,<at 0.6 2>]}}}, Scale Curve=2)
      B=&FloatFromCurve(FloatCurve=DataInterface() {propertyValues {Curve {keys [<at 0 1.3 RCIM_Constant>,<at 0.2 -1.7 RCIM_Constant>,<at 0.6 2>]}}}, Scale Curve=-2)
      Alpha=&MakeCustomFloatFromBool(Bool=&RandomBool(), True Float=1, False Float=0)
     )
    )) {enabled=false}
   &Drag(Use Linear Drag=true, Use Rotational Drag=false)
   &GravityForce(Gravity=&Multiply_VectorByFloat(
     Vector=&Multiply_VectorByFloat(Vector=<0 0 -20>, Float=2.2)
     Float=&FloatFromCurve(FloatCurve=DataInterface() {propertyValues {Curve {keys [<at 0.2 0>,<at 0.6 1>]}}})
    ))
   &InheritSourceMovement(Applied Position Scale=<0.5 0.5 0>, ApplyForce=false, ApplyPosition=true, ApplyPositionFalloff=false)
   &ScaleSpriteSize(Uniform Curve Sprite Scale=DataInterface() {propertyValues {Curve {keys [
        {interpMode=RCIM_Cubic, leaveTangentWeight=0.0698, tangentMode=RCTM_User, time=0, value=0.5}
        {arriveTangentWeight=0.1582, interpMode=RCIM_Cubic, leaveTangentWeight=0.1204, tangentMode=RCTM_User, time=0.2, value=1.5}
        {arriveTangentWeight=0.0695, interpMode=RCIM_Cubic, leaveTangentWeight=0.1889, tangentMode=RCTM_User, time=0.35, value=1}
        {arriveTangentWeight=0.1727, interpMode=RCIM_Cubic, leaveTangentWeight=0.0399, tangentMode=RCTM_User, time=0.7, value=1}
        {arriveTangent=-0.0447, interpMode=RCIM_Cubic, leaveTangent=-0.0447, tangentMode=RCTM_User, time=0.85, value=1.525}
        {arriveTangent=-10.4508, interpMode=RCIM_Cubic, leaveTangent=-10.4508, tangentMode=RCTM_User, time=1, value=0.5}
       ]}}})
   &ScaleSpriteSize(Uniform Curve Sprite Scale=DataInterface() {propertyValues {Curve {keys [
        {arriveTangent=6.2934, interpMode=RCIM_Cubic, leaveTangent=6.2934, leaveTangentWeight=0.0167, tangentMode=RCTM_User, time=0, value=1}
        {arriveTangentWeight=0.0113, interpMode=RCIM_Cubic, leaveTangentWeight=0.0053, tangentMode=RCTM_User, time=0.034, value=1.15}
        <at 0.05 1.15>
        {
         arriveTangent=-1.0657
         arriveTangentWeight=0.0286
         interpMode=RCIM_Cubic
         leaveTangent=0.0038
         leaveTangentWeight=0.0547
         tangentMode=RCTM_Break
         time=0.1359
         value=1
        }
        {
         arriveTangent=-0.0068
         arriveTangentWeight=0.0129
         interpMode=RCIM_Cubic
         leaveTangent=4.5082
         leaveTangentWeight=0.0418
         tangentMode=RCTM_Break
         time=0.1747
         value=1
        }
        {arriveTangentWeight=0.0113, interpMode=RCIM_Cubic, leaveTangentWeight=0.0304, tangentMode=RCTM_User, time=0.2087, value=1.15}
        <at 0.3 1>
        {
         arriveTangent=-1.0657
         arriveTangentWeight=0.0035
         interpMode=RCIM_Cubic
         leaveTangent=0.0035
         leaveTangentWeight=0.0131
         tangentMode=RCTM_Break
         time=0.3106
         value=1
        }
        {
         arriveTangent=-0.0012
         arriveTangentWeight=0.0129
         interpMode=RCIM_Cubic
         leaveTangent=5.4011
         leaveTangentWeight=0.0002
         tangentMode=RCTM_Break
         time=0.3495
         value=1
        }
        <at 0.35 1>
        {arriveTangentWeight=0.0111, interpMode=RCIM_Cubic, leaveTangentWeight=0.0055, tangentMode=RCTM_User, time=0.3834, value=1.15}
        <at 0.4 1.15>
        {
         arriveTangent=-1.1038
         arriveTangentWeight=0.0285
         interpMode=RCIM_Cubic
         leaveTangent=0.0027
         leaveTangentWeight=0.0715
         tangentMode=RCTM_Break
         time=0.4854
         value=1
        }
        {
         arriveTangent=-0.0003
         arriveTangentWeight=0.0113
         interpMode=RCIM_Cubic
         leaveTangent=7.5323
         leaveTangentWeight=0.0602
         tangentMode=RCTM_Break
         time=0.5193
         value=1
        }
        {arriveTangentWeight=0.0113, interpMode=RCIM_Cubic, leaveTangentWeight=0.0489, tangentMode=RCTM_User, time=0.5533, value=1.15}
        {
         arriveTangent=-1.0657
         arriveTangentWeight=0.034
         interpMode=RCIM_Cubic
         leaveTangent=-0.004
         leaveTangentWeight=0.0149
         tangentMode=RCTM_Break
         time=0.6552
         value=1
        }
        {
         arriveTangent=0.0028
         arriveTangentWeight=0.0129
         interpMode=RCIM_Cubic
         leaveTangent=3.1075
         leaveTangentWeight=0.002
         tangentMode=RCTM_Break
         time=0.6941
         value=1
        }
        <at 0.7 1>
        {arriveTangentWeight=0.0093, interpMode=RCIM_Cubic, leaveTangentWeight=0.024, tangentMode=RCTM_User, time=0.728, value=1.15}
        <at 0.8 1>
        {
         arriveTangent=-1.0657
         arriveTangentWeight=0.01
         interpMode=RCIM_Cubic
         leaveTangent=0.0025
         leaveTangentWeight=0.0233
         tangentMode=RCTM_Break
         time=0.83
         value=1
        }
        {
         arriveTangent=-0.0003
         arriveTangentWeight=0.0129
         interpMode=RCIM_Cubic
         leaveTangent=6.0131
         leaveTangentWeight=0.0104
         tangentMode=RCTM_Break
         time=0.8688
         value=1
        }
       ]}}}) {module=ScaleSpriteSize001}
   /SetVariables(Particles.GlowColor=&MultiplyLinearColor(
     Linear Color=&MultiplyLinearColor(
      Linear Color=$Particles.Initial.GlowColor
      Scale Factor=&MakeVector4_FromFloat(Float=&FloatFromCurve(FloatCurve=DataInterface() {propertyValues {Curve {keys [<at 0.9 1>,<at 1 0>]}}}))
     )
     Scale Factor=&MakeVector4_FromFloat(Float=&FNTimeOfDayScalar(Night Scale=0.5, Morning Scale=1, Evening Scale=0.5))
    )) {isSetParameters=true, module=SetVariables_#2}
   &SolveForcesAndVelocity(Clamp Velocity=false, Limit Acceleration=false)
   &GenerateDeathEvent(Event Probability=1, Use Event Probability=true)
   &ScaleSpriteSize(
    Scale Sprite Size Mode=::NewEnumerator0
    Uniform Scale Factor=&FN_LOD_Distance_Remap(Start Distance=400, End Distance=900, Minimum Output=1, Maximum Output=1.2)
   ) {module=ScaleSpriteSize002}
   &ScaleColor(
    Scale RGB=&VectorFromFloat(Value=&FNTimeOfDayScalar(Night Scale=0.5, Morning Scale=1, Evening Scale=0.3))
    ScaleA=true
    ScaleRGB=true
   )
  ]
  rendererClasses [&NiagaraSpriteRendererProperties]
  renderers [
   {rendererClass=&NiagaraSpriteRendererProperties}
   {rendererClass=&NiagaraSpriteRendererProperties, rendererIndex=1}
  ]
 }
 Hearts001 {
  bEnabled=false
  emitterUpdateScript [
   &EmitterState(Life Cycle Mode=::NewEnumerator1, Loop Delay=0.9083, Loop Duration=0, Recalculate Duration Each Loop=true, UseLoopDelay=false)
   &SpawnRate(SpawnRate=1.5, Use Spawn Probability=false) {module=SpawnRate001}
   &SpawnRate(SpawnGroup=1, SpawnRate=1.5, Use Spawn Probability=false)
  ]
  particleSpawnScript [
   &InitializeParticle(
    Color=<rgba 0 0.4255 1 1>
    Lifetime=&RandomRangeFloat(Minimum=2, Maximum=2.5)
    Sprite Size=&Multiply_Vector2D_ByFloat(
     Float=&FloatFromCurve(
      FloatCurve=DataInterface() {propertyValues {Curve {keys [<at 0 0.65 RCIM_Constant>,<at 0.45 0.9 RCIM_Constant>,<at 0.9 1.3>]}}}
      CurveIndex=&RandomRangeFloat(Minimum=0, Maximum=1)
      Scale Curve=1.2
     )
     Vector2D=<7 3>
    )
    Sprite Size Mode=::NewEnumerator4
    UsePositionOffset=false
    Write Material Random=true
   )
   &SystemLocation(
    MaskSpawnGroup=false
    Offset=&Lerp_Vector(A=<-1 2 3.5>, B=<-1 2 -3.5>, Alpha=&MaskFloatBySpawnGroup(Masked Float=0, Passthrough Float=1))
   )
   /SetVariables(
    Particles.GlowColor=&MultiplyLinearColor(Linear Color=$Particles.Color, Scale Factor=&MakeVector4_FromFloat(Float=0.2))
    Particles.GlowSize=&Multiply_Vector2D_ByFloat(
     Float=0.9
     Vector2D=&Vector2DFromFloat(Value=&MakeFloatFromVector2d(Vector 2D=$Particles.SpriteSize))
    )
   ) {isSetParameters=true, module=SetVariables_#1}
   &AddVelocity(
    Cone Angle=60
    Cone Axis=&Lerp_Vector(A=<0 0 1>, B=<0 0 -1>, Alpha=&MaskFloatBySpawnGroup(Masked Float=0, Passthrough Float=1))
    Distribution Along Cone Axis=0.25
    Inner Cone Angle=10
    Module.UseConeFalloff=true
    Module.UseVelDistribution=true
    Override Seed=false
    Speed Falloff From Cone Axis=0.5
    Velocity Mode=::NewEnumerator2
    Velocity Speed=&RandomRangeFloat(Minimum=17, Maximum=17)
   )
   &SubUVAnimation(End Frame=1, SubUV Animation Mode:ENiagara_SubUVLookupMode=::NewEnumerator2)
  ]
  particleUpdateScript [
   &ParticleState()
   &AccelerationForce(Acceleration=&MakeVector(
     X=0
     Y=0
     Z=&Lerp_Float(
      A=&FloatFromCurve(FloatCurve=DataInterface() {propertyValues {Curve {keys [<at 0 1.3 RCIM_Constant>,<at 0.2 -1.7 RCIM_Constant>,<at 0.6 2>]}}}, Scale Curve=2)
      B=&FloatFromCurve(FloatCurve=DataInterface() {propertyValues {Curve {keys [<at 0 1.3 RCIM_Constant>,<at 0.2 -1.7 RCIM_Constant>,<at 0.6 2>]}}}, Scale Curve=-2)
      Alpha=&MakeCustomFloatFromBool(Bool=&RandomBool(), True Float=1, False Float=0)
     )
    )) {enabled=false}
   &Drag(
    Drag=&FloatFromCurve(FloatCurve=DataInterface() {propertyValues {Curve {keys [<at 0.8 0.5>,<at 1 0.8>]}}})
    Use Linear Drag=true
    Use Rotational Drag=false
   )
   &GravityForce(Gravity=&Multiply_VectorByFloat(
     Vector=<0 0 -20>
     Float=&FloatFromCurve(FloatCurve=DataInterface() {propertyValues {Curve {keys [<at 0.2 0>,<at 0.6 1>]}}})
    ))
   &InheritSourceMovement(Applied Position Scale=<0.5 0.5 0>, ApplyForce=false, ApplyPosition=true, ApplyPositionFalloff=false)
   &ScaleSpriteSize(Uniform Curve Sprite Scale=DataInterface() {propertyValues {Curve {keys [
        {interpMode=RCIM_Cubic, leaveTangentWeight=0.0698, tangentMode=RCTM_User, time=0, value=0}
        {arriveTangentWeight=0.1582, interpMode=RCIM_Cubic, leaveTangentWeight=0.1204, tangentMode=RCTM_User, time=0.2, value=1.5}
        {arriveTangentWeight=0.0695, interpMode=RCIM_Cubic, leaveTangentWeight=0.1889, tangentMode=RCTM_User, time=0.35, value=1}
        {arriveTangentWeight=0.1727, interpMode=RCIM_Cubic, leaveTangentWeight=0.0399, tangentMode=RCTM_User, time=0.8015, value=1}
        <at 0.9 1.525>
        {arriveTangent=-10.4508, interpMode=RCIM_Cubic, leaveTangent=-10.4508, tangentMode=RCTM_User, time=1, value=0.5}
       ]}}})
   &ScaleSpriteSize(Uniform Curve Sprite Scale=DataInterface() {propertyValues {Curve {keys [
        {arriveTangent=6.2934, interpMode=RCIM_Cubic, leaveTangent=6.2934, leaveTangentWeight=0.0167, tangentMode=RCTM_User, time=0, value=1}
        {arriveTangentWeight=0.0113, interpMode=RCIM_Cubic, leaveTangentWeight=0.0053, tangentMode=RCTM_User, time=0.034, value=1.15}
        <at 0.05 1.15>
        {
         arriveTangent=-1.0657
         arriveTangentWeight=0.0286
         interpMode=RCIM_Cubic
         leaveTangent=0.0038
         leaveTangentWeight=0.0547
         tangentMode=RCTM_Break
         time=0.1359
         value=1
        }
        {
         arriveTangent=-0.0068
         arriveTangentWeight=0.0129
         interpMode=RCIM_Cubic
         leaveTangent=4.5082
         leaveTangentWeight=0.0418
         tangentMode=RCTM_Break
         time=0.1747
         value=1
        }
        {arriveTangentWeight=0.0113, interpMode=RCIM_Cubic, leaveTangentWeight=0.0304, tangentMode=RCTM_User, time=0.2087, value=1.15}
        <at 0.3 1>
        {
         arriveTangent=-1.0657
         arriveTangentWeight=0.0035
         interpMode=RCIM_Cubic
         leaveTangent=0.0035
         leaveTangentWeight=0.0131
         tangentMode=RCTM_Break
         time=0.3106
         value=1
        }
        {
         arriveTangent=-0.0012
         arriveTangentWeight=0.0129
         interpMode=RCIM_Cubic
         leaveTangent=5.4011
         leaveTangentWeight=0.0002
         tangentMode=RCTM_Break
         time=0.3495
         value=1
        }
        <at 0.35 1>
        {arriveTangentWeight=0.0111, interpMode=RCIM_Cubic, leaveTangentWeight=0.0055, tangentMode=RCTM_User, time=0.3834, value=1.15}
        <at 0.4 1.15>
        {
         arriveTangent=-1.1038
         arriveTangentWeight=0.0285
         interpMode=RCIM_Cubic
         leaveTangent=0.0027
         leaveTangentWeight=0.0715
         tangentMode=RCTM_Break
         time=0.4854
         value=1
        }
        {
         arriveTangent=-0.0003
         arriveTangentWeight=0.0113
         interpMode=RCIM_Cubic
         leaveTangent=7.5323
         leaveTangentWeight=0.0602
         tangentMode=RCTM_Break
         time=0.5193
         value=1
        }
        {arriveTangentWeight=0.0113, interpMode=RCIM_Cubic, leaveTangentWeight=0.0489, tangentMode=RCTM_User, time=0.5533, value=1.15}
        {
         arriveTangent=-1.0657
         arriveTangentWeight=0.034
         interpMode=RCIM_Cubic
         leaveTangent=-0.004
         leaveTangentWeight=0.0149
         tangentMode=RCTM_Break
         time=0.6552
         value=1
        }
        {
         arriveTangent=0.0028
         arriveTangentWeight=0.0129
         interpMode=RCIM_Cubic
         leaveTangent=3.1075
         leaveTangentWeight=0.002
         tangentMode=RCTM_Break
         time=0.6941
         value=1
        }
        <at 0.7 1>
        {arriveTangentWeight=0.0093, interpMode=RCIM_Cubic, leaveTangentWeight=0.024, tangentMode=RCTM_User, time=0.728, value=1.15}
        <at 0.8 1>
        {
         arriveTangent=-1.0657
         arriveTangentWeight=0.01
         interpMode=RCIM_Cubic
         leaveTangent=0.0025
         leaveTangentWeight=0.0233
         tangentMode=RCTM_Break
         time=0.83
         value=1
        }
        {
         arriveTangent=-0.0003
         arriveTangentWeight=0.0129
         interpMode=RCIM_Cubic
         leaveTangent=6.0131
         leaveTangentWeight=0.0104
         tangentMode=RCTM_Break
         time=0.8688
         value=1
        }
        <at 0.9 1.15>
        {arriveTangentWeight=0.0009, interpMode=RCIM_Cubic, leaveTangentWeight=0.0324, tangentMode=RCTM_User, time=0.9028, value=1.15}
        <at 0.9707 1>
        {tangentMode=RCTM_User, time=1, value=1}
       ]}}}) {module=ScaleSpriteSize001}
   /SetVariables(Particles.GlowColor=&MultiplyLinearColor(
     Linear Color=&MultiplyLinearColor(
      Linear Color=$Particles.Initial.GlowColor
      Scale Factor=&MakeVector4_FromFloat(Float=&FloatFromCurve(FloatCurve=DataInterface() {propertyValues {Curve {keys [<at 0.9 1>,<at 1 0>]}}}))
     )
     Scale Factor=&MakeVector4_FromFloat(Float=&FNTimeOfDayScalar(Night Scale=0.5, Morning Scale=1, Evening Scale=1))
    )) {isSetParameters=true, module=SetVariables_#2}
   &SolveForcesAndVelocity(Clamp Velocity=false, Limit Acceleration=false)
   &GenerateDeathEvent(Event Probability=1, Use Event Probability=true)
   &ScaleColor(
    Scale RGB=&VectorFromFloat(Value=&FNTimeOfDayScalar(Night Scale=0.5, Morning Scale=1, Evening Scale=1))
    ScaleA=true
    ScaleRGB=true
   )
  ]
  rendererClasses [&NiagaraSpriteRendererProperties]
  renderers [
   {rendererClass=&NiagaraSpriteRendererProperties}
   {rendererClass=&NiagaraSpriteRendererProperties, rendererIndex=1}
  ]
 }
 Pops {
  emitterUpdateScript [&EmitterState(Life Cycle Mode=::NewEnumerator1, Loop Duration=0, Recalculate Duration Each Loop=true, UseLoopDelay=false)]
  particleSpawnScript [
   &InitializeParticle(
    Color=<rgba 0.9834 0.3594 1 0.5>
    Lifetime=0.2
    Mass Mode=::NewEnumerator3
    Sprite Rotation Angle=&Lerp_Float(
     A=45
     B=0
     Alpha=&MakeCustomFloatFromBool(Bool=&RandomBool(), True Float=1, False Float=0)
    )
    Sprite Rotation Mode=::NewEnumerator0
    Sprite Size Mode=::NewEnumerator3
    Uniform Sprite Size=7
    UsePositionOffset=false
    Write Material Random=true
   )
   /SetVariables(
    Particles.GlowColor=&MultiplyLinearColor(Linear Color=<rgba 0.3 0.0703 0.2599 1>, Scale Factor=&MakeVector4_FromFloat(Float=0.25))
    Particles.GlowSize=&Multiply_Vector2D_ByFloat(
     Float=2.5
     Vector2D=&MakeVector2D(
      X=&MakeFloatFromVector2d(Vector 2D=$Particles.SpriteSize)
      Y=&MakeFloatFromVector2d(Vector 2D=$Particles.SpriteSize)
     )
    )
   ) {isSetParameters=true}
  ]
  particleUpdateScript [
   /SetVariables(Particles.GlowColor=&MultiplyLinearColor(
     Linear Color=&MultiplyLinearColor(
      Linear Color=$Particles.Initial.GlowColor
      Scale Factor=&MakeVector4_FromFloat(Float=&FloatFromCurve(FloatCurve=DataInterface() {propertyValues {Curve {keys [<at 0 1>,<at 1 0>]}}}))
     )
     Scale Factor=&MakeVector4_FromFloat(Float=&FNTimeOfDayScalar(Night Scale=1, Morning Scale=1, Evening Scale=1))
    )) {isSetParameters=true}
   &ParticleState()
   /SetVariables(Particles.Color=&MultiplyLinearColor(
     Linear Color=$Particles.Initial.Color
     Scale Factor=&MakeVector4_FromLinearColor(LinearColor=&MakeLinearColorFromVectorAndFloat(
       "Vector (RGB)"=&VectorFromFloat(Value=&FNTimeOfDayScalar(Night Scale=0.7, Morning Scale=1, Evening Scale=0.5))
       "Float (Alpha)"=1
      ))
    )) {isSetParameters=true}
   &ScaleSpriteSize(Uniform Curve Sprite Scale=DataInterface() {propertyValues {Curve {keys [{tangentMode=RCTM_User, time=0, value=0.7}, {tangentMode=RCTM_User, time=1, value=1.4}]}}})
   &SolveForcesAndVelocity(Clamp Velocity=false, Limit Acceleration=false)
   &SubUVAnimation(End Frame=4)
  ]
  rendererClasses [&NiagaraSpriteRendererProperties]
  renderers [
   {rendererClass=&NiagaraSpriteRendererProperties}
   {rendererClass=&NiagaraSpriteRendererProperties, rendererIndex=1}
  ]
 }
}
summary {
 systemName=NS_PinkSpike_Emote
 emitters [
  {emitterName=Hearts, rendererClasses [&NiagaraSpriteRendererProperties]}
  {emitterName=Pops, rendererClasses [&NiagaraSpriteRendererProperties]}
  {emitterName=Flare, rendererClasses [&NiagaraSpriteRendererProperties]}
  {bEnabled=false, emitterName=Hearts001, rendererClasses [&NiagaraSpriteRendererProperties]}
 ]
 userVariables [
  User.bIsFrontEndPreview {type=NiagaraBool}
  User.FrontendSpawn {type=NiagaraBool}
 ]
}
